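# One-shot mono MetaSound source for the Steel Kite consumable, playing the sound for the end of its active state.
# On play, a single Wave Player plays the SteelKite_Active_End wave to the mono output and fires OnFinished when it completes.
# It is assigned to the Fort_Consumables sound class and uses the Athena consumable-use attenuation.
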MetaSoundSource /WildEstateConsumables/Gameplay/SteelKite/Audio/MSS_SteelKite_Active_End
inputs [UE.Source.OnPlay {type=Trigger}]
interfaces [UE.OutputFormat.Mono:1.0, UE.Source.OneShot:1.0, UE.Source:1.1]
outputs [
 "UE.OutputFormat.Mono.Audio:0" {type=Audio}
 UE.Source.OneShot.OnFinished {type=Trigger}
]
pages [Default {
  input_defaults {UE.Source.OnPlay=null}
  nodes [
   InitVariable.WaveAsset:Array()
   WavePlayer_0 = UE.Wave Player:Mono(Play=^UE.Source.OnPlay) {also {defaults {Wave Asset=/WildEstateConsumables/Gameplay/SteelKite/Audio/Waves/SteelKite_Active_End}, version [1,1]}}
  ]
  links [{dst="@UE.Source.OneShot.OnFinished", src=WavePlayer_0.On Finished}, {dst="@UE.OutputFormat.Mono.Audio:0", src=WavePlayer_0.Out Mono}]
 }]
properties {
 assetImportData=&self:AssetImportData
 attenuationSettings=&Fort_Consumable_Use_Athena_Att
 maxDistance=2250
 sampleRate=48000
 soundClassObject=&Fort_Consumables
 volume=0.75
}
