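# Looping mono MetaSound source for the first-person audio of the active Steel Kite consumable, layering an idle loop with a speed-driven movement loop.
# The idle loop starts on play, while MSP_OneshotGate watches Movement.Speed to trigger an ADSR envelope's attack and release.
# The envelope fades the movement loop, which restarts at a random offset on each attack and stops when done.
# FadeOutTime drives the attack time, and FadeInTime and Loop Sound are unconnected.
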MetaSoundSource /WildEstateConsumables/Gameplay/SteelKite/Audio/MSS_Steelkite_Active_1p
inputs [
 ActiveLoop {type=WaveAsset}
 FadeInTime {type=Float}
 FadeOutTime {type=Time}
 Loop Sound {type=WaveAsset}
 Movement.Speed {type=Float}
 Movement.SpeedAzimuth {type=Float}
 Movement.SpeedElevation {type=Float}
 SpeedLoop {type=WaveAsset}
 UE.Source.OnPlay {type=Trigger}
]
interfaces [Fort.Movement:1.0, UE.OutputFormat.Mono:1.0, UE.Source:1.1]
outputs ["UE.OutputFormat.Mono.Audio:0" {type=Audio}]
pages [Default {
  input_defaults {
   ActiveLoop=/WildEstateConsumables/Gameplay/SteelKite/Audio/Waves/SteelKite_Active_Idle_Loop
   FadeInTime=1
   FadeOutTime=1
   Loop Sound=null
   Movement.Speed=1
   Movement.SpeedAzimuth=0
   Movement.SpeedElevation=0
   SpeedLoop=/WildEstateConsumables/Gameplay/SteelKite/Audio/Waves/SteelKite_Active_Movement_Loop
   UE.Source.OnPlay=null
  }
  nodes [
   n1 = &MSP_OneshotGate(Value=^Movement.Speed) {also {defaults {OneshotThreshold=1}}}
   ADSREnvelopeADSREnvelope_0 = ADSR Envelope.ADSR Envelope:Float(Attack Time=^FadeOutTime, Trigger Attack=@n1.OnGateEnter, Trigger Release=@n1.OnGateExit) {also {defaults {Decay Time=0, Release Time=5, Sustain Level=1}}}
   GetWaveDuration_0 = "UE.Get Wave Duration: "(Wave=^SpeedLoop)
   WavePlayer_1 = UE.Wave Player:Mono(
    Wave Asset=^SpeedLoop
    Play=@ADSREnvelopeADSREnvelope_0.On Attack Triggered
    Stop=@ADSREnvelopeADSREnvelope_0.On Done
    Start Time=UE.RandomTime(Next=@ADSREnvelopeADSREnvelope_0.On Attack Triggered, Max=@GetWaveDuration_0.Duration).Value
   ) {also {defaults {Loop="<MetaSoundBuilderResult.SUCCEEDED: 0>"}, version [1,1]}}
   AudioMixerAudioMixerMono2_0 = "AudioMixer.Audio Mixer (Mono, 2)"(In 1=@WavePlayer_1.Out Mono) {also {defaults {Gain 1=0.5}}}
   WavePlayer_0 = UE.Wave Player:Mono(Wave Asset=^ActiveLoop, Play=^UE.Source.OnPlay) {also {defaults {Loop="<MetaSoundBuilderResult.SUCCEEDED: 0>"}, version [1,1]}}
   AudioMixerAudioMixerMono2_1 = "AudioMixer.Audio Mixer (Mono, 2)"(Gain 1=@ADSREnvelopeADSREnvelope_0.Out Envelope, In 1=@AudioMixerAudioMixerMono2_0.Out, In 0=@WavePlayer_0.Out Mono)
  ]
  links [{dst="@UE.OutputFormat.Mono.Audio:0", src=AudioMixerAudioMixerMono2_1.Out}]
 }]
properties {
 assetImportData=&self:AssetImportData
 bLooping=true
 duration=10000
 sampleRate=48000
 soundClassObject=&Fort_Consumables
 volume=0.5
}
