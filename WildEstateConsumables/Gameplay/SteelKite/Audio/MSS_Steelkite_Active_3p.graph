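# Mono MetaSound source for the active Steel Kite consumable's looping sound, named as the third-person variant.
# A single wave player starts on play and feeds the active idle loop wave straight to the mono output.
# Fort movement inputs and a fade time are declared but left unconnected, so neither affects the sound.
# It uses consumable attenuation and sound class, with overridden concurrency capping simultaneous instances.
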MetaSoundSource /WildEstateConsumables/Gameplay/SteelKite/Audio/MSS_Steelkite_Active_3p
inputs [
 ActiveLoop {type=WaveAsset}
 FadeTime {type=Time}
 Movement.Speed {type=Float}
 Movement.SpeedAzimuth {type=Float}
 Movement.SpeedElevation {type=Float}
 UE.Source.OnPlay {type=Trigger}
]
interfaces [Fort.Movement:1.0, UE.OutputFormat.Mono:1.0, UE.Source:1.1]
outputs ["UE.OutputFormat.Mono.Audio:0" {type=Audio}]
pages [Default {
  input_defaults {
   ActiveLoop=/WildEstateConsumables/Gameplay/SteelKite/Audio/Waves/SteelKite_Active_Idle_Loop
   FadeTime=0.5
   Movement.Speed=0
   Movement.SpeedAzimuth=0
   Movement.SpeedElevation=0
   UE.Source.OnPlay=null
  }
  nodes [WavePlayer_0 = UE.Wave Player:Mono(Wave Asset=^ActiveLoop, Play=^UE.Source.OnPlay) {also {defaults {Loop="<MetaSoundBuilderResult.SUCCEEDED: 0>"}, version [1,1]}}]
  links [{dst="@UE.OutputFormat.Mono.Audio:0", src=WavePlayer_0.Out Mono}]
 }]
properties {
 assetImportData=&self:AssetImportData
 attenuationSettings=&Fort_Consumable_Use_Athena_Att
 bLooping=true
 bOverrideConcurrency=true
 duration=10000
 maxDistance=2250
 sampleRate=48000
 soundClassObject=&Fort_Consumables
 volume=0.75
 concurrencyOverrides {maxCount=4}
 concurrencySet [null]
}
