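; Steel Kite consumable weapon Blueprint that shows a frosty wisp ice-trail effect while the item is held.
; The equipped-and-attached event activates the NS_IceTrail_FrostyWisps component with a reset, and the unequipped-and-detached event deactivates it.
; ShouldShowSoundIndicator always returns false, and no other gameplay logic is defined in this graph.
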(event EventBeginPlay
  (|CallParentFunction) ; This node is disabled and will not be called.\nDrag off pins to build functionality.
)

(event EventTick
  (params (DeltaSeconds :type Float%20%28single-precision%29))
  (|CallParentFunction DeltaSeconds) ; This node is disabled and will not be called.\nDrag off pins to build functionality.
)

(event FortWeapon|EventOnUnEquippedAndDetached
  (Components|Activation|Deactivate (Variables|Default|GetNS_IceTrail_FrostyWisps)))

(event FortWeapon|EventOnEquippedAndAttached
  (Components|Activation|Activate (Variables|Default|GetNS_IceTrail_FrostyWisps) true))

(fn ShouldShowSoundIndicator
  (params (PlayerController :type Fort%20Player%20Controller%20Object%20Reference))
  (return false))

(fn ConstructionScript
  (|CallParentFunction))
