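# Translucent unlit Niagara particle material that draws a scrolling rainbow glow outline for the SteelKite effect.
# FXCommon_SoftOutline builds a custom-depth mask, which offsets time-scrolled rainbow texture lookups and, contrast-shaped, blends in a start color.
# Hue, saturation and brightness adjustments and particle color feed emissive, while depth fade, inverted Fresnel and particle alpha shape opacity.
# Dynamic parameters expose outline, scroll speed and blend shaping to particles.
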Material /WildEstateConsumables/Gameplay/SteelKite/FX/M_BodyGlowRainbow
graph {
 properties {
  bUsedWithNiagaraMeshParticles=true
  bUsedWithNiagaraSprites=true
  blendMode=BLEND_Translucent
  pixelDepthOffsetMode=PDOM_Legacy
  shadingModel=MSM_Unlit
  twoSided=true
  usedShadingModels=Unlit
 }
 nodes [
  DynamicParameter() {parameterIndex=3, paramNames [NA,NA,NA,NA]}
  n1 = VectorParameter() {parameterName=HSV Low, defaultValue=<rgba 0 2 1.5 0>}
  n2 = DynamicParameter() {defaultValue=<rgba 2.7 150 1 100>, paramNames [Outline thickness, Max Blur Depth, Enable Inner Glow, Inner Glow Max Area]}
  n3 = DynamicParameter() {parameterIndex=1, defaultValue=<rgba 10 0.5 1 -0.75>, paramNames [Inner Glow Pow, Inner GLow Strength, Outline Str, Color Speed]}
  n4 = MaterialFunctionCall!(
   GlowMeshStencilIndex=ScalarParameter() {defaultValue=40, parameterName="GlowMeshStencilIndex (S)"}
   UseCustomDepth=StaticBoolParameter() {defaultValue=true, parameterName=UseCustomDepth}
   UseCustomStencil=StaticBoolParameter() {parameterName=UseCustomStencil}
   OutlineThickness=@n2.Outline thickness
   OutlineStrength=@n3.Outline Str
   PixelDepthOffset
   MaxBlurDepth=@n2.Outline thickness
   ------EnableInnerGlow------=@n2.Outline thickness
   InnerGlowMaxArea=@n2.Outline thickness
   InnerGlowPower=@n3.Outline Str
   InnerGlowStrength=@n3.Outline Str
   ------LessUse------
   MeshPosition
   Distance Steps
   Radial Steps
  ) {materialFunction=/Game/Effects/Fort_Effects/Effects/Atlas/Materials/FXCommon_SoftOutline}
  n5 = Multiply(A=@n4.Result, B=ScalarParameter() {parameterName=Tile})
  n6 = Time()
  n7 = OneMinus(None=Frac(None=Multiply(A=@n6, B=@n3.Color Speed)))
  n8 = DynamicParameter() {parameterIndex=2, paramNames [Pow,Mul,Add,Contrast]}
  n9 = Multiply(A=Power(Base=@n4.Result, Exp=@n8.Pow), B=@n8.Mul)
  n10 = LinearInterpolate(
   A=TextureSample(UVs=Add(A=@n5, B=@n7)) {texture=/WildEstateConsumables/Gameplay/SteelKite/FX/Textures/Rainbow_Colors_basecolor}.RGB
   B=VectorParameter() {parameterName=Start Color, defaultValue=<rgba 1 0.2 0 0>}.RGB
   Alpha=MaterialFunctionCall(In=Add(A=@n9, B=@n8.Add) {constB=0}, Contrast=@n8.Contrast) {materialFunction=&CheapContrast}.Result
  )
  n11 = Multiply(
   A=Desaturation(
    None=MaterialFunctionCall(Hue Shift Percentage=@n1.R, Texture=@n10) {materialFunction=&HueShift}.Result
    Fraction=OneMinus(None=@n1.G)
   )
   B=@n1.B
  )
  LinearInterpolate(
   A=@n11
   B=@n10
   Alpha=Round(None=MaterialFunctionCall() {materialFunction=&DebugTimeSine}.Result)
  )
  n12 = ParticleColor()
  MaterialFunctionCall(Emissive=Divide(A=Multiply(A=@n11, B=@n12.RGB))) {materialFunction=&MF_FX_EmissiveAdjustments}
  n13 = ScalarParameter() {defaultValue=100, parameterName=DepthFade}
  n14 = OneMinus(None=Fresnel(ExponentIn=ScalarParameter() {defaultValue=1, parameterName=Fresnel_Exponent}) {baseReflectFraction=0, exponent=1})
  n15 = ScalarParameter() {defaultValue=2, parameterName=Fresnel_Power}
  Multiply(A=Multiply(A=DepthFade(Opacity=@n4.Result, FadeDistance=@n13), B=@n12.A), B=Saturate(None=Power(Base=@n14, Exp=@n15)))
 ]
}
