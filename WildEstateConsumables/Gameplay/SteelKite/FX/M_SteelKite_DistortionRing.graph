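# Translucent Niagara sprite material for the SteelKite distortion ring effect, using index-of-refraction distortion with depth testing disabled.
# A mip-biased ring texture supplies alpha, whose screen-space derivatives (DDX/DDY) form a derived normal.
# That alpha also blends low and high refraction parameters, with particle random scaling the result from neutral.
# Particle color and alpha multiply the texture's red and alpha channels.
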Material /WildEstateConsumables/Gameplay/SteelKite/FX/M_SteelKite_DistortionRing
graph {
 properties {
  bDisableDepthTest=true
  bUsedWithNiagaraSprites=true
  blendMode=BLEND_Translucent
  refractionMethod=RM_IndexOfRefraction
  usedShadingModels=Default Lit
 }
 nodes [
  n1 = TextureSample!(UVs, Tex, Bias=ScalarParameter() {defaultValue=4.5, parameterName=Bias}, Apply View MipBias) {mipValueMode=TMVM_MipBias, texture=/WildEstateConsumables/Gameplay/SteelKite/FX/Textures/T_SteelKite_DistortionRing}
  n2 = DDX(Value=@n1.A)
  DeriveNormalZ(InXY=AppendVector(A=@n2, B=DDY(Value=@n1.A)))
  LinearInterpolate(
   B=LinearInterpolate(
    A=ScalarParameter() {defaultValue=1, parameterName=Refract Low}
    B=ScalarParameter() {defaultValue=1.4, parameterName=Refract High}
    Alpha=@n1.A
   )
   Alpha=ParticleRandom()
  ) {constA=1}
  n3 = ParticleColor()
  Multiply(A=@n1.R, B=@n3.RGB)
  Multiply(A=@n1.A, B=@n3.A)
  Multiply(A=@n2) {constB=-1}
  Multiply(A=@n1.A)
 ]
}
