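# Translucent unlit material for Niagara sprite and mesh particles in the SteelKite FX, set up for 2D-offset refraction jitter.
# A dynamic-parameter Jitter X vector is scaled by a depth-fade term and a UV sphere mask; Jitter Y and SM Hardness are unused.
# The depth-fade lerp takes Low at both ends, cancelling the fade.
# A separate contrast-adjusted vertical gradient feeds nothing, and no material output connections appear.
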Material /WildEstateConsumables/Gameplay/SteelKite/FX/M_jitter_refraction
graph {
 properties {
  bAutomaticallySetUsageInEditor=false
  bUsedWithNiagaraMeshParticles=true
  bUsedWithNiagaraSprites=true
  blendMode=BLEND_Translucent
  refractionMethod=RM_2DOffset
  shadingModel=MSM_Unlit
  usedShadingModels=Unlit
 }
 nodes [
  Ceil()
  Constant3Vector() {constant=<rgba 1 0 1 0>}
  Constant()
  Constant() {r=1}
  n1 = VectorParameter() {parameterName=Depth Fade Options, channelNames {a="", b=High, g=Low, r=Fade Dist}, defaultValue=<rgba 100 0 1 0>}
  n2 = LinearInterpolate(A=@n1.Low, B=@n1.Low, Alpha=DepthFade(FadeDistance=@n1.Fade Dist))
  n3 = DynamicParameter() {paramNames [Jitter X, Jitter Y, SM Rad, SM Hardness]}
  Multiply(
   A=Multiply(A=Saturate(None=@n2), B=AppendVector(A=@n3.Jitter X, B=@n3.Jitter X))
   B=SphereMask(A=TextureCoordinate(), B=Constant() {r=0.5}, Radius=@n3.SM Rad, Hardness=@n3.SM Rad)
  )
  n4 = OneMinus(None=MaterialFunctionCall() {materialFunction=&LinearGradient}.VGradient)
  n5 = VectorParameter() {parameterName=gradient options, channelNames {a=Contrast, b=add, g=mul, r=pow}, defaultValue=<rgba 1 1 0 0>}
  n6 = Add(A=Multiply(A=Power(Base=@n4, Exp=@n5.pow) {constExponent=1}, B=@n5.mul), B=@n5.add) {constB=0}
  Multiply(A=Saturate(None=MaterialFunctionCall(In=@n6, Contrast=@n5.Contrast) {materialFunction=&CheapContrast}.Result))
  ParticleColor()
  SceneTexture() {sceneTextureId=PPI_CustomDepth}
  TextureCoordinate()
 ]
}
