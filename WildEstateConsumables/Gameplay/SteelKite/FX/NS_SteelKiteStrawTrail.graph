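# Niagara ribbon trail for the Steel Kite's straw, drawn as a short-lived orange-brown streak.
# One emitter spawns particles per frame (per-unit spawning is disabled), which inherit a small share of velocity and fall under doubled gravity.
# Alpha follows a curve indexed by User.NNP rather than particle age: opaque at 0.5 and below, fading to invisible at the default of 1.
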NiagaraSystem /WildEstateConsumables/Gameplay/SteelKite/FX/NS_SteelKiteStrawTrail
dependencies {usedDataInterfaces [&NiagaraDataInterfaceCurve], usedDynamicInputs [], usedModules [&SystemState]}
emitters {Ribbon {
  emitterUpdateScript [
   &EmitterState()
   &SpawnPerFrame(Use Spawn Probability=false)
   &SpawnPerUnit(Module.Use Max Movement Threshold=false, Module.Use Movement Tolerance=false, Use Spawn Probability=false) {enabled=false}
  ]
  particleSpawnScript [
   &InitializeParticle(
    Color=<rgba 0.5 0.1945 0 1>
    Lifetime=0.125
    Ribbon Width=1
    Ribbon Width Mode=::NewEnumerator1
    UsePositionOffset=false
    Write Material Random=true
   )
   &InheritVelocity(Inherited Velocity Amount Scale=&VectorFromFloat(Value=0.1), Speed Limit=true, UseSpeedThreshold=false)
  ]
  particleUpdateScript [
   &ParticleState()
   &GravityForce(Gravity=&Multiply_VectorByFloat(Vector=<0 0 -980>, Float=2))
   &SolveForcesAndVelocity(Clamp Velocity=false, Limit Acceleration=false)
   &ScaleColor(
    Scale Alpha=&FloatFromCurve(
     FloatCurve=DataInterface() {propertyValues {Curve {keys [<at 0.5 1 RCIM_Cubic>, {arriveTangent=-3.5479, interpMode=RCIM_Cubic, leaveTangent=-3.5479, tangentMode=RCTM_User, time=1, value=0}]}}}
     CurveIndex=$User.NNP
    )
    ScaleA=true
    ScaleRGB=true
   )
  ]
  rendererClasses [&NiagaraRibbonRendererProperties]
  renderers [{rendererClass=&NiagaraRibbonRendererProperties}]
 }}
summary {
 systemName=NS_SteelKiteStrawTrail
 emitters [{emitterName=Ribbon, rendererClasses [&NiagaraRibbonRendererProperties]}]
 userVariables [User.NNP {type=NiagaraFloat, defaultValue=1, description=""}]
}
