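# Niagara effect for the SteelKite consumable, named as a lens effect on consume, that shows a large, faint amber sprite burst.
# A single GPU emitter burst-spawns camera-offset sprites that live two seconds, growing from zero along an ease-out curve while alpha fades out.
# Particle MaterialRandom is overwritten each update by a scaled curve that spikes early and decays, repurposing it as an animated material input.
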NiagaraSystem /WildEstateConsumables/Gameplay/SteelKite/FX/NS_SteelKite_ConsumeLensEffect
dependencies {usedDataInterfaces [&NiagaraDataInterfaceCurve], usedDynamicInputs [], usedModules [&SystemState]}
emitters {OmnidirectionalBurst {
  simTarget=GPUComputeSim
  emitterUpdateScript [
   &EmitterState(Life Cycle Mode=::NewEnumerator1, Loop Behavior=::NewEnumerator1, Loop Duration=0.25, UseLoopDelay=false)
   &SpawnBurst_Instantaneous(Use Spawn Probability=false, UseLoopCountLimit=false)
  ]
  particleSpawnScript [
   &InitializeParticle(
    Color=<rgba 0.4621 0.231 0 0.1>
    Lifetime=2
    Sprite Rotation Mode=::NewEnumerator2
    Sprite Size Mode=::NewEnumerator3
    Uniform Sprite Size=600
    UsePositionOffset=false
    Write Material Random=true
   )
   &CameraOffset(Camera Offset Amount=-25)
  ]
  particleUpdateScript [
   &ParticleState()
   &ScaleSpriteSize(Uniform Curve Sprite Scale=DataInterface() {propertyValues {Curve {keys [
        {
         arriveTangent=3.453
         arriveTangentWeight=1.0254
         interpMode=RCIM_Cubic
         leaveTangent=3.453
         leaveTangentWeight=1.0254
         tangentMode=RCTM_User
         tangentWeightMode=RCTWM_WeightedBoth
         time=0
         value=0
        }
        {
         arriveTangent=0.165
         arriveTangentWeight=0.4167
         interpMode=RCIM_Cubic
         leaveTangent=0.165
         leaveTangentWeight=0.4167
         tangentMode=RCTM_User
         tangentWeightMode=RCTWM_WeightedBoth
         time=1
         value=1
        }
       ]}}})
   &ScaleColor(
    Scale Alpha=&FloatFromCurve(FloatCurve=DataInterface() {propertyValues {Curve {keys [<at -0.0046 1.0113 RCIM_Cubic>,<at 1.0011 0.002 RCIM_Cubic>]}}})
    ScaleA=true
    ScaleRGB=true
   )
   /SetVariables(Particles.MaterialRandom:NiagaraFloat=&FloatFromCurve(
     FloatCurve=DataInterface() {propertyValues {Curve {
        defaultValue=1
        keys [
         {arriveTangent=9.0725, interpMode=RCIM_Cubic, leaveTangent=9.0725, tangentMode=RCTM_User, time=0, value=0}
         {interpMode=RCIM_Cubic, leaveTangent=-3.2313, tangentMode=RCTM_Break, time=0.1408, value=1}
         {
          arriveTangent=-0.0407
          arriveTangentWeight=0.5176
          interpMode=RCIM_Cubic
          leaveTangent=-0.0407
          leaveTangentWeight=0.5176
          tangentMode=RCTM_User
          tangentWeightMode=RCTWM_WeightedBoth
          time=1
          value=0
         }
        ]
       }}}
     Scale Curve=0.2
    )) {isSetParameters=true}
  ]
  rendererClasses [&NiagaraSpriteRendererProperties]
  renderers [{rendererClass=&NiagaraSpriteRendererProperties}]
 }}
summary {
 systemName=NS_SteelKite_ConsumeLensEffect
 emitters [{emitterName=OmnidirectionalBurst, simTarget=GPUComputeSim, rendererClasses [&NiagaraSpriteRendererProperties]}]
 userVariables []
}
