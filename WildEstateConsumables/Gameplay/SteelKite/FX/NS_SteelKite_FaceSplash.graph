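# Niagara effect for the SteelKite consumable's face splash, combining a simulated fluid burst with flying drips.
# A GPU PIC/FLIP emitter runs on a camera-facing 2D grid at the owner, using render targets and thickness-based surfacing, tinted blue.
# Four drip emitters add orange SubUV sprites and tumbling meshes, thrown out in cones under drag and gravity.
# Drips briefly inherit horizontal source movement, and the fluid emitter culls by instance significance.
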NiagaraSystem /WildEstateConsumables/Gameplay/SteelKite/FX/NS_SteelKite_FaceSplash
dependencies {
 usedDataInterfaces [
  &NiagaraDataInterfaceArrayFloat
  &NiagaraDataInterfaceSpriteRendererInfo
  &NiagaraDataInterfaceCurve
  &NiagaraDataInterfaceVector2DCurve
  &NiagaraDataInterfaceArrayInt32
  /Script/Niagara.NiagaraDataInterfaceGrid2DCollection
  /Script/Niagara.NiagaraDataInterfaceRenderTarget2D
  &NiagaraDataInterfaceCamera
  /Script/Niagara.NiagaraDataInterfaceRasterizationGrid3D
  &NiagaraDataInterfacePlatformSet
  /Script/Niagara.NiagaraDataInterfaceEmitterProperties
  &NiagaraDataInterfaceSkeletalMesh
 ]
 usedDynamicInputs [&MakeVector, &SelectFloatFromArray]
 usedModules [
  &self:SystemSpawnScript.SystemScriptSource.SystemScriptGraph.NiagaraNodeAssignment_0.SetVariables_#1
  &SystemState
 ]
}
emitters {
 Drips {
  emitterUpdateScript [
   &EmitterState()
   &SpawnRate(SpawnRate=10, Use Spawn Probability=false)
   &SpawnBurst_Instantaneous(Spawn Count=50, Use Spawn Probability=false, UseLoopCountLimit=false) {enabled=false}
   &SpawnBurst_Instantaneous(
    Spawn Count=&RandomRangeInt(Minimum=1, Maximum=2)
    Spawn Group=1
    Use Spawn Probability=false
    UseLoopCountLimit=false
   ) {module=SpawnBurst_Instantaneous002}
  ]
  particleSpawnScript [
   &InitializeParticle(
    Color=<rgba 1.25 0.5208 0 1.5>
    Lifetime=&RandomRangeFloat(Minimum=0.45, Maximum=0.2)
    Sprite Rotation Mode=::NewEnumerator2
    Sprite Size Max=<7 25>
    Sprite Size Min=<5 35>
    Sprite Size Mode=::NewEnumerator2
    UsePositionOffset=false
    Write Material Random=true
   )
   &CameraOffset(Camera Offset Amount=50)
   &AddVelocity(
    Cone Angle=180
    Cone Axis=<0 0 1>
    Inner Cone Angle=150
    Module.UseConeFalloff=false
    Module.UseVelDistribution=false
    Override Seed=false
    Velocity Mode=::NewEnumerator2
    Velocity Speed=&RandomRangeFloat(Minimum=200, Maximum=250)
   )
   &SubUVAnimation(
    "Sprite Renderer "=DataInterface() {propertyValues {SpriteRenderer=&self:Drips_0.NiagaraSpriteRendererProperties_2}}
    SubUV Animation Mode:ENiagara_SubUVLookupModeV2=::NewEnumerator2
    UseEndFrame=false
    UseStartFrame=false
   )
   &SystemLocation(MaskSpawnGroup=false, Offset=<0 0 3.5>)
  ]
  particleUpdateScript [
   &ParticleState()
   &ScaleSpriteSize(
    Scale Factor=&Vector2DFromFloat(Value=&FloatFromCurve(
      FloatCurve=DataInterface() {propertyValues {Curve {keys [
          {arriveTangent=0.9171, interpMode=RCIM_Cubic, leaveTangent=0.9171, tangentMode=RCTM_User, time=0, value=0.75}
          {arriveTangent=-0.75, interpMode=RCIM_Cubic, leaveTangent=-0.75, time=0.6325, value=1.0524}
          {arriveTangent=-2.9974, interpMode=RCIM_Cubic, leaveTangent=-2.9974, tangentMode=RCTM_User, time=1, value=0}
         ]}}}
      Scale Curve=0.25
     ))
    Scale Sprite Size Mode=::NewEnumerator2
   )
   &ScaleColor(
    Scale Alpha=&MaskFloatBySpawnGroup(
     Masked Float=&FloatFromCurve(
      FloatCurve=DataInterface() {propertyValues {Curve {keys [
          {arriveTangent=-0.2723, interpMode=RCIM_Cubic, leaveTangent=-0.2723, tangentMode=RCTM_User, time=0, value=1}
          {arriveTangent=-3.7547, interpMode=RCIM_Cubic, leaveTangent=-3.7547, tangentMode=RCTM_User, time=1, value=0}
         ]}}}
      Scale Curve=2
     )
     Passthrough Float=1
     Spawn Group Mask=2
    )
    ScaleA=true
    ScaleRGB=false
   ) {enabled=false}
   &Drag(Drag=5, Use Linear Drag=true, Use Rotational Drag=false)
   &GravityForce()
   &InheritSourceMovement(
    Applied Position Scale=&Multiply_VectorByFloat(
     Vector=<1 1 0>
     Float=&FloatFromCurve(FloatCurve=DataInterface() {propertyValues {Curve {keys [<at 0 1>,<at 0.1 0>]}}}, CurveIndex=$Engine.System.Age)
    )
    ApplyForce=false
    ApplyPosition=true
    ApplyPositionFalloff=false
   )
   &SolveForcesAndVelocity(Clamp Velocity=false, Limit Acceleration=false)
   &DynamicMaterialParameters(
    Index 0 Param 1=&FloatFromCurve(FloatCurve=DataInterface() {propertyValues {Curve {keys [
         {arriveTangent=-0.0467, interpMode=RCIM_Cubic, leaveTangent=-0.0467, tangentMode=RCTM_User, time=0.1, value=1}
         {arriveTangent=-1.2111, tangentMode=RCTM_User, time=1, value=0}
        ]}}})
    Module.Index0Param0Write=true
    Module.Index0Param1Write=false
    Module.Index0Param2Write=false
    Module.Index0Param3Write=false
   )
  ]
  rendererClasses [&NiagaraSpriteRendererProperties]
  renderers [{rendererClass=&NiagaraSpriteRendererProperties}]
 }
 Drips001 {
  emitterUpdateScript [
   &EmitterState()
   &SpawnRate(SpawnGroup=1, SpawnRate=15, Use Spawn Probability=false)
   &SpawnBurst_Instantaneous(Spawn Count=50, Use Spawn Probability=false, UseLoopCountLimit=false) {enabled=false}
   &SpawnBurst_Instantaneous(Spawn Count=3, Spawn Group=1, Use Spawn Probability=false, UseLoopCountLimit=false) {module=SpawnBurst_Instantaneous002}
  ]
  particleSpawnScript [
   /SetVariables(
    Particles.Variance:NiagaraFloat=&RandomRangeFloat(Minimum=0, Maximum=1)
    Particles.VisibilityTag=&SelectIntFromArray(Direct Array Index=&SpawnGroup(), Int Selection Array=DataInterface() {propertyValues {IntData [0,1,1]}})
   ) {isSetParameters=true, module=SetVariables_#2}
   &InitializeParticle(
    Color=<rgba 1 0.3667 0 5>
    Lifetime=&RandomRangeFloat(Minimum=0.45, Maximum=0.3)
    Sprite Rotation Mode=::NewEnumerator2
    Sprite Size Max=<35 35>
    Sprite Size Min=<25 25>
    Sprite Size Mode=::NewEnumerator2
    UsePositionOffset=false
    Write Material Random=true
   )
   &CameraOffset(Camera Offset Amount=10)
   &AddVelocity(
    Rotation Coordinate Space=::World
    Velocity=&RandomRangeVector(Minimum=<-250 -250 200>, Maximum=<250 250 250>)
   ) {enabled=false}
   &AddVelocity(
    Cone Angle=180
    Cone Axis=<0 0 1>
    Inner Cone Angle=150
    Module.UseConeFalloff=false
    Module.UseVelDistribution=false
    Override Seed=false
    Rotation Coordinate Space=::World
    Velocity Mode=::NewEnumerator2
    Velocity Speed=&RandomRangeFloat(Minimum=50, Maximum=10)
   ) {module=AddVelocity001}
   &SubUVAnimation(
    "Sprite Renderer "=DataInterface() {propertyValues {SpriteRenderer=&self:Drips001_0.NiagaraSpriteRendererProperties_2}}
    SubUV Animation Mode:ENiagara_SubUVLookupModeV2=::NewEnumerator2
    UseEndFrame=false
    UseStartFrame=false
   )
  ]
  particleUpdateScript [
   &ParticleState()
   &SpriteRotationRate(Rotation Rate=&Multiply_Float(A=360, B=&RandomRangeFloat(Minimum=-1, Maximum=1))) {enabled=false}
   &ScaleSpriteSize(
    Scale Factor=&MaskVector2dBySpawnGroup(
     Masked Vector2D=&Vector2DFromCurve(
      Vector2Curve=DataInterface() {propertyValues {
        XCurve {keys [
          <at -0.0013 0.0604 RCIM_Cubic>
          {arriveTangent=1.8744, interpMode=RCIM_Cubic, leaveTangent=1.8744, time=0.29, value=1.0472}
          {arriveTangent=-1.4749, interpMode=RCIM_Cubic, leaveTangent=-1.4749, time=0.5, value=1}
          {arriveTangent=-0.4016, tangentMode=RCTM_User, time=1, value=0}
         ]}
        YCurve {keys [
          {arriveTangent=3.3463, interpMode=RCIM_Cubic, leaveTangent=3.3463, tangentMode=RCTM_User, time=0, value=0}
          <at 0.3 1>
          {arriveTangent=-1.281, interpMode=RCIM_Cubic, leaveTangent=-1.281, tangentMode=RCTM_User, time=0.5, value=1}
          {arriveTangent=-0.1464, tangentMode=RCTM_User, time=1, value=0}
         ]}
       }}
      ScaleCurve=<1 1>
     )
     Passthrough Vector2D=&Vector2DFromFloat(Value=&FloatFromCurve(FloatCurve=DataInterface() {propertyValues {Curve {keys [
           {arriveTangent=-0.412, interpMode=RCIM_Cubic, time=0.3588, value=1}
           {arriveTangent=-2.4131, arriveTangentWeight=0.2925, interpMode=RCIM_Cubic, leaveTangent=-2.4131, tangentMode=RCTM_User, time=1, value=0.15}
          ]}}}))
     Spawn Group Mask=0
    )
    Scale Sprite Size Mode=::NewEnumerator2
   ) {enabled=false}
   &ScaleColor(
    Scale Alpha=&MaskFloatBySpawnGroup(
     Masked Float=&FloatFromCurve(
      FloatCurve=DataInterface() {propertyValues {Curve {keys [
          {arriveTangent=-0.2723, interpMode=RCIM_Cubic, leaveTangent=-0.2723, tangentMode=RCTM_User, time=0, value=0}
          {arriveTangentWeight=0.032, interpMode=RCIM_Cubic, leaveTangentWeight=0.3013, tangentMode=RCTM_User, time=0.2, value=0.9779}
          {arriveTangent=-3.7547, interpMode=RCIM_Cubic, leaveTangent=-3.7547, tangentMode=RCTM_User, time=1, value=0}
         ]}}}
      Scale Curve=2
     )
     Passthrough Float=1
     Spawn Group Mask=2
    )
    ScaleA=true
    ScaleRGB=false
   ) {enabled=false}
   &Drag(Drag=3, Use Linear Drag=true, Use Rotational Drag=false)
   &GravityForce()
   &SampleSkeletalMesh(AdditionalYawPitchRoll=false, OverrideRandomSeed=false, Skeletal Mesh:NiagaraDataInterfaceSkeletalMesh=$System.Skeletal Mesh Reader) {enabled=false}
   &InheritSourceMovement(
    Applied Position Scale=&Multiply_VectorByFloat(
     Vector=<1 1 0>
     Float=&FloatFromCurve(FloatCurve=DataInterface() {propertyValues {Curve {keys [<at 0 1>,<at 0.1 0>]}}}, CurveIndex=$Engine.System.Age)
    )
    ApplyForce=false
    ApplyPosition=true
    ApplyPositionFalloff=false
   )
   &SolveForcesAndVelocity(Clamp Velocity=false, Limit Acceleration=false)
   &DynamicMaterialParameters(
    Index 0 Param 1=&FloatFromCurve(
     FloatCurve=DataInterface() {propertyValues {Curve {
        defaultValue=0
        keys [
         <at 0 0 RCIM_Cubic>
         <at 0.0804 1 RCIM_Cubic>
         {arriveTangent=-0.3689, interpMode=RCIM_Cubic, leaveTangent=-0.3689, tangentMode=RCTM_User, time=1, value=0}
        ]
       }}}
     Scale Curve=&RandomRangeFloat(Minimum=0.5, Maximum=0.7)
    )
    Module.Index0Param0Write=true
    Module.Index0Param1Write=false
    Module.Index0Param2Write=false
    Module.Index0Param3Write=false
   )
  ]
  rendererClasses [&NiagaraSpriteRendererProperties]
  renderers [{rendererClass=&NiagaraSpriteRendererProperties}]
 }
 Drips002 {
  emitterUpdateScript [
   &EmitterState()
   &SpawnRate(SpawnRate=50, Use Spawn Probability=false)
   &SpawnBurst_Instantaneous(Spawn Count=50, Use Spawn Probability=false, UseLoopCountLimit=false) {enabled=false}
   &SpawnBurst_Instantaneous(Spawn Count=15, Spawn Group=1, Use Spawn Probability=false, UseLoopCountLimit=false) {module=SpawnBurst_Instantaneous002}
  ]
  particleSpawnScript [
   &InitializeParticle(
    Color=<rgba 1 0.3667 0 5>
    Lifetime=&RandomRangeFloat(Minimum=0.5, Maximum=0.7)
    Mesh Scale Max=<1.2 1.2 1.2>
    Mesh Scale Min=<0.8 0.8 0.8>
    Mesh Scale Mode=::NewEnumerator2
    Sprite Rotation Mode=::NewEnumerator2
    Sprite Size Max=<35 35>
    Sprite Size Min=<25 25>
    Sprite Size Mode=::NewEnumerator2
    UsePositionOffset=false
    Write Material Random=true
   )
   &CameraOffset()
   &AddVelocity(
    Rotation Coordinate Space=::World
    Velocity=&RandomRangeVector(Minimum=<-250 -250 200>, Maximum=<250 250 250>)
   ) {enabled=false}
   &AddVelocity(
    Cone Angle=180
    Cone Axis=<0 0 1>
    Inner Cone Angle=150
    Module.UseConeFalloff=false
    Module.UseVelDistribution=false
    Override Seed=false
    Rotation Coordinate Space=::World
    Velocity Mode=::NewEnumerator2
    Velocity Speed=&RandomRangeFloat(Minimum=50, Maximum=150)
   ) {module=AddVelocity001}
   &InitialMeshOrientation(Use Rotation Vector=false)
   /SetVariables(Particles.float:NiagaraFloat=&RandomRangeFloat(Minimum=0, Maximum=1000)) {isSetParameters=true}
  ]
  particleUpdateScript [
   &ParticleState()
   &ScaleMeshSize(Scale Factor=&VectorFromFloat(Value=&FloatFromCurve(
      FloatCurve=DataInterface() {propertyValues {Curve {keys [<at 0 1 RCIM_Cubic>, {arriveTangent=-3.2601, interpMode=RCIM_Cubic, leaveTangent=-3.2601, tangentMode=RCTM_User, time=1, value=0}]}}}
      Scale Curve=&RandomRangeFloat(Minimum=0.4, Maximum=0.6)
     )))
   &UpdateMeshOrientation(
    Rotation Rate=&RandomRangeFloat(Minimum=0.5, Maximum=2)
    Rotation Vector=&RandomVector(Vector Scale=1)
   )
   &ScaleColor(
    Scale Alpha=&MaskFloatBySpawnGroup(
     Masked Float=&FloatFromCurve(
      FloatCurve=DataInterface() {propertyValues {Curve {keys [
          {arriveTangent=-0.2723, interpMode=RCIM_Cubic, leaveTangent=-0.2723, tangentMode=RCTM_User, time=0, value=0}
          {arriveTangentWeight=0.032, interpMode=RCIM_Cubic, leaveTangentWeight=0.3013, tangentMode=RCTM_User, time=0.2, value=0.9779}
          {arriveTangent=-3.7547, interpMode=RCIM_Cubic, leaveTangent=-3.7547, tangentMode=RCTM_User, time=1, value=0}
         ]}}}
      Scale Curve=2
     )
     Passthrough Float=1
     Spawn Group Mask=2
    )
    ScaleA=true
    ScaleRGB=false
   )
   &Drag(Drag=3, Use Linear Drag=true, Use Rotational Drag=false)
   &GravityForce(Gravity=&Multiply_VectorByFloat(Vector=<0 0 -980>, Float=&RandomRangeFloat(Minimum=0.25, Maximum=1)))
   &InheritSourceMovement(
    Applied Position Scale=&Multiply_VectorByFloat(
     Vector=<1 1 0>
     Float=&FloatFromCurve(FloatCurve=DataInterface() {propertyValues {Curve {keys [<at 0 1>,<at 0.5 0>]}}}, CurveIndex=$Engine.System.Age)
    )
    ApplyForce=false
    ApplyPosition=true
    ApplyPositionFalloff=false
   )
   &SolveForcesAndVelocity(Clamp Velocity=false, Limit Acceleration=false)
   &DynamicMaterialParameters(
    Index 0 Vector 1:Vector3f=&Multiply_VectorByFloat(Vector=&MakeVectorFromLinearColorRGB(Color=$Particles.Color), Float=0.75)
    Module.Index0Float3Write=true
    Module.Index0VectorWrite=true
    Write Parameter Index 0=::NewEnumerator3
   )
  ]
  rendererClasses [&NiagaraMeshRendererProperties]
  renderers [{rendererClass=&NiagaraMeshRendererProperties, rendererIndex=0}]
 }
 Drips003 {
  emitterUpdateScript [
   &EmitterState()
   &SpawnRate(SpawnGroup=1, SpawnRate=15, Use Spawn Probability=false)
   &SpawnBurst_Instantaneous(Spawn Count=50, Use Spawn Probability=false, UseLoopCountLimit=false) {enabled=false}
   &SpawnBurst_Instantaneous(Spawn Count=3, Spawn Group=1, Use Spawn Probability=false, UseLoopCountLimit=false) {module=SpawnBurst_Instantaneous002}
  ]
  particleSpawnScript [
   /SetVariables(
    Particles.Variance:NiagaraFloat=&RandomRangeFloat(Minimum=0, Maximum=1)
    Particles.VisibilityTag=&SelectIntFromArray(Direct Array Index=&SpawnGroup(), Int Selection Array=DataInterface() {propertyValues {IntData [0,1,1]}})
   ) {isSetParameters=true, module=SetVariables_#2}
   &InitializeParticle(
    Color=<rgba 1 0.3667 0 5>
    Lifetime=&RandomRangeFloat(Minimum=0.45, Maximum=0.3)
    Sprite Rotation Mode=::NewEnumerator2
    Sprite Size Max=<35 35>
    Sprite Size Min=<25 25>
    Sprite Size Mode=::NewEnumerator2
    UsePositionOffset=false
    Write Material Random=true
   )
   &CameraOffset(Camera Offset Amount=10)
   &AddVelocity(
    Rotation Coordinate Space=::World
    Velocity=&RandomRangeVector(Minimum=<-250 -250 200>, Maximum=<250 250 250>)
   ) {enabled=false}
   &AddVelocity(
    Cone Angle=180
    Cone Axis=<0 0 1>
    Inner Cone Angle=150
    Module.UseConeFalloff=false
    Module.UseVelDistribution=false
    Override Seed=false
    Rotation Coordinate Space=::World
    Velocity Mode=::NewEnumerator2
    Velocity Speed=&RandomRangeFloat(Minimum=50, Maximum=10)
   ) {module=AddVelocity001}
   &SubUVAnimation(
    "Sprite Renderer "=DataInterface() {propertyValues {SpriteRenderer=&self:Drips003_0.NiagaraSpriteRendererProperties_2}}
    SubUV Animation Mode:ENiagara_SubUVLookupModeV2=::NewEnumerator2
    UseEndFrame=false
    UseStartFrame=false
   )
  ]
  particleUpdateScript [
   &ParticleState()
   &SpriteRotationRate(Rotation Rate=&Multiply_Float(A=360, B=&RandomRangeFloat(Minimum=-1, Maximum=1))) {enabled=false}
   &ScaleSpriteSize(
    Scale Factor=&MaskVector2dBySpawnGroup(
     Masked Vector2D=&Vector2DFromCurve(
      Vector2Curve=DataInterface() {propertyValues {
        XCurve {keys [
          <at -0.0013 0.0604 RCIM_Cubic>
          {arriveTangent=1.8744, interpMode=RCIM_Cubic, leaveTangent=1.8744, time=0.29, value=1.0472}
          {arriveTangent=-1.4749, interpMode=RCIM_Cubic, leaveTangent=-1.4749, time=0.5, value=1}
          {arriveTangent=-0.4016, tangentMode=RCTM_User, time=1, value=0}
         ]}
        YCurve {keys [
          {arriveTangent=3.3463, interpMode=RCIM_Cubic, leaveTangent=3.3463, tangentMode=RCTM_User, time=0, value=0}
          <at 0.3 1>
          {arriveTangent=-1.281, interpMode=RCIM_Cubic, leaveTangent=-1.281, tangentMode=RCTM_User, time=0.5, value=1}
          {arriveTangent=-0.1464, tangentMode=RCTM_User, time=1, value=0}
         ]}
       }}
      ScaleCurve=<1 1>
     )
     Passthrough Vector2D=&Vector2DFromFloat(Value=&FloatFromCurve(FloatCurve=DataInterface() {propertyValues {Curve {keys [
           {arriveTangent=-0.412, interpMode=RCIM_Cubic, time=0.3588, value=1}
           {arriveTangent=-2.4131, arriveTangentWeight=0.2925, interpMode=RCIM_Cubic, leaveTangent=-2.4131, tangentMode=RCTM_User, time=1, value=0.15}
          ]}}}))
     Spawn Group Mask=0
    )
    Scale Sprite Size Mode=::NewEnumerator2
   ) {enabled=false}
   &ScaleColor(
    Scale Alpha=&MaskFloatBySpawnGroup(
     Masked Float=&FloatFromCurve(
      FloatCurve=DataInterface() {propertyValues {Curve {keys [
          {arriveTangent=-0.2723, interpMode=RCIM_Cubic, leaveTangent=-0.2723, tangentMode=RCTM_User, time=0, value=0}
          {arriveTangentWeight=0.032, interpMode=RCIM_Cubic, leaveTangentWeight=0.3013, tangentMode=RCTM_User, time=0.2, value=0.9779}
          {arriveTangent=-3.7547, interpMode=RCIM_Cubic, leaveTangent=-3.7547, tangentMode=RCTM_User, time=1, value=0}
         ]}}}
      Scale Curve=2
     )
     Passthrough Float=1
     Spawn Group Mask=2
    )
    ScaleA=true
    ScaleRGB=false
   ) {enabled=false}
   &Drag(Drag=3, Use Linear Drag=true, Use Rotational Drag=false)
   &GravityForce()
   &SampleSkeletalMesh(AdditionalYawPitchRoll=false, OverrideRandomSeed=false, Skeletal Mesh:NiagaraDataInterfaceSkeletalMesh=$System.Skeletal Mesh Reader) {enabled=false}
   &InheritSourceMovement(
    Applied Position Scale=&Multiply_VectorByFloat(
     Vector=<1 1 0>
     Float=&FloatFromCurve(FloatCurve=DataInterface() {propertyValues {Curve {keys [<at 0 1>,<at 0.1 0>]}}}, CurveIndex=$Engine.System.Age)
    )
    ApplyForce=false
    ApplyPosition=true
    ApplyPositionFalloff=false
   )
   &SolveForcesAndVelocity(Clamp Velocity=false, Limit Acceleration=false)
   &DynamicMaterialParameters(
    Index 0 Param 1=&FloatFromCurve(
     FloatCurve=DataInterface() {propertyValues {Curve {
        defaultValue=0
        keys [
         <at 0 0 RCIM_Cubic>
         <at 0.0804 1 RCIM_Cubic>
         {arriveTangent=-0.3689, interpMode=RCIM_Cubic, leaveTangent=-0.3689, tangentMode=RCTM_User, time=1, value=0}
        ]
       }}}
     Scale Curve=&RandomRangeFloat(Minimum=0.5, Maximum=0.7)
    )
    Module.Index0Param0Write=true
    Module.Index0Param1Write=false
    Module.Index0Param2Write=false
    Module.Index0Param3Write=false
   )
  ]
  rendererClasses [&NiagaraSpriteRendererProperties]
  renderers [{rendererClass=&NiagaraSpriteRendererProperties}]
 }
 FN_PIC_FLIP_Fluid_Stream002 {
  simTarget=GPUComputeSim
  emitterSpawnScript [
   /SetVariables(
    Emitter.PressureGrid=DataInterface() {propertyValues {
      ClearBeforeNonIterationStage=true
      NumAttributes=0
      NumCellsMaxAxis=3
      NumCellsX=3
      NumCellsY=3
      OverrideBufferFormat=Float
      PreviewAttribute=None
      SetGridFromMaxAxis=false
      bOverrideFormat=true
      bPreviewGrid=false
      RenderTargetUserParameter {parameter {name=None, typeDefHandle {registeredTypeIndex=4}, varData []}}
      WorldBBoxSize=<100 100>
     }}
    Emitter.VelocityGrid=DataInterface() {propertyValues {
      ClearBeforeNonIterationStage=true
      NumAttributes=0
      NumCellsMaxAxis=3
      NumCellsX=128
      NumCellsY=64
      OverrideBufferFormat=Float
      PreviewAttribute=None
      SetGridFromMaxAxis=false
      bOverrideFormat=false
      bPreviewGrid=false
      RenderTargetUserParameter {parameter {name=None, typeDefHandle {registeredTypeIndex=4}, varData []}}
      WorldBBoxSize=<100 100>
     }}
   ) {isSetParameters=true}
   /NiagaraFluids/Modules/Grid2D/Grid2D_SetResolution(
    Grid:NiagaraDataInterfaceGrid2DCollection=$Emitter.VelocityGrid
    NumCellsMaxAxis=256
    Resolution Method:ENiagaraGrid2DResolution=::NewEnumerator1
    WorldGridExtents:Vector2f=<256 256>
   )
   &Grid2D_SetResolution(
    Grid:NiagaraDataInterfaceGrid2DCollection=$Emitter.PressureGrid
    OtherGrid:NiagaraDataInterfaceGrid2DCollection=$Emitter.VelocityGrid
    Resolution Method:ENiagaraGrid2DResolution=::NewEnumerator3
    WorldGridExtents:Vector2f=$Emitter.Grid2D_SetResolution.WorldGridExtents
   ) {module=Grid2D_SetResolution002}
   /SetVariables(Emitter.DeltaTime=0.033) {isSetParameters=true}
   /SetVariables(Emitter.SimRT=DataInterface() {propertyValues {
      MipMapGeneration=Disabled
      MipMapGenerationType=Linear
      OverrideRenderTargetFilter=TF_Default
      OverrideRenderTargetFormat=RTF_RGBA16f
      bInheritUserParameterSettings=false
      bOverrideFormat=true
      bPreviewRenderTarget=false
      RenderTargetUserParameter {parameter {name=None, typeDefHandle {registeredTypeIndex=4}, varData []}}
      Size=<0 0>
     }}) {isSetParameters=true}
   /NiagaraFluids/Modules/Grid2D/Grid2D_InitRT(Grid:NiagaraDataInterfaceGrid2DCollection=$Emitter.VelocityGrid, SimRT:NiagaraDataInterfaceRenderTarget2D=$Emitter.SimRT) {module=InitRTGrid2D}
   &self:FN_PIC_FLIP_Fluid_Stream002_2.NiagaraScratchPadContainer_0.ScratchModule_0() {module=ScratchModule}
   &self:FN_PIC_FLIP_Fluid_Stream002_2.NiagaraScratchPadContainer_0.ScratchModule_03_5(NumPressureIterations=8) {module=ScratchModule_03}
   /SetVariables() {isSetParameters=true}
   /SetVariables(Emitter.ThicknessRT=DataInterface() {propertyValues {
      MipMapGeneration=Disabled
      MipMapGenerationType=Linear
      OverrideRenderTargetFilter=TF_Default
      OverrideRenderTargetFormat=RTF_R16f
      bInheritUserParameterSettings=false
      bOverrideFormat=true
      bPreviewRenderTarget=false
      RenderTargetUserParameter {parameter {name=None, typeDefHandle {registeredTypeIndex=4}, varData []}}
      Size=<0 0>
     }}) {isSetParameters=true}
   &Grid2D_InitRT(Grid:NiagaraDataInterfaceGrid2DCollection=$Emitter.VelocityGrid, SimRT:NiagaraDataInterfaceRenderTarget2D=$Emitter.ThicknessRT)
   /SetVariables(
    Emitter.RasterizationGrid=DataInterface() {propertyValues {
      CellSize=1
      ClearBeforeNonIterationStage=true
      NumAttributes=1
      NumCellsMaxAxis=10
      Precision=1000
      ResetValue=100000000
      SetResolutionMethod=Independent
      NumCells=<3 3 3>
      WorldBBoxSize=<100 100 100>
     }}
    Emitter.RasterizationGridBack=DataInterface() {propertyValues {
      CellSize=1
      ClearBeforeNonIterationStage=true
      NumAttributes=1
      NumCellsMaxAxis=10
      Precision=1000
      ResetValue=0
      SetResolutionMethod=Independent
      NumCells=<3 3 3>
      WorldBBoxSize=<100 100 100>
     }}
   ) {isSetParameters=true}
   /NiagaraFluids/Modules/Grid3D/RasterizationGrid3D_SetResolution(
    Grid:NiagaraDataInterfaceRasterizationGrid3D=$Emitter.RasterizationGrid
    NumCellsX:NiagaraInt32=/Game/Effects/FlipSimulation/Niagara/DynamicInputs/GetGrid2DSize(Grid2D Collection:NiagaraDataInterfaceGrid2DCollection=$Emitter.VelocityGrid, Y:NiagaraBool=false)
    NumCellsY:NiagaraInt32=&GetGrid2DSize(Grid2D Collection:NiagaraDataInterfaceGrid2DCollection=$Emitter.VelocityGrid, Y:NiagaraBool=true)
    NumCellsZ=1
    WorldGridExtents:Vector3f=<1 1 1>
   )
   &RasterizationGrid3D_SetResolution(
    Grid:NiagaraDataInterfaceRasterizationGrid3D=$Emitter.RasterizationGridBack
    NumCellsX:NiagaraInt32=&GetGrid2DSize(Grid2D Collection:NiagaraDataInterfaceGrid2DCollection=$Emitter.VelocityGrid, Y:NiagaraBool=false)
    NumCellsY:NiagaraInt32=&GetGrid2DSize(Grid2D Collection:NiagaraDataInterfaceGrid2DCollection=$Emitter.VelocityGrid, Y:NiagaraBool=true)
    NumCellsZ=1
    WorldGridExtents:Vector3f=<1 1 1>
   ) {module=RasterizationGrid3D_SetResolution001}
   /SetVariables(Emitter.CameraOffset=5) {isSetParameters=true}
  ]
  emitterUpdateScript [
   /NiagaraFN/Modules/Scalability/SignificantInstancesCull(Cull Response:ENiagaraExecutionStateManagement=::KillImmediately)
   &EmitterState()
   /NiagaraFluids/Modules/Grid2D/Grid2D_UpdateSimAttributes(Grid:NiagaraDataInterfaceGrid2DCollection=$Emitter.VelocityGrid, WorldGridSize:Vector2f=$Emitter.Grid2D_SetResolution.WorldGridExtents) {module=UpdateSimAttributes}
   /NiagaraFluids/Modules/Grid2D/Grid2D_CreateUnitToWorldTransform(
    Camera:NiagaraDataInterfaceCamera=$Emitter.CameraQuery
    Camera Facing=true
    Local Pivot:Vector2f=<0.5 0.65>
    Offset:NiagaraPosition=$Engine.Owner.Position
    Rotation:Quat4f=&self:FN_PIC_FLIP_Fluid_Stream002_2.NiagaraScratchPadContainer_0.ScratchDynamicInput_01_1(Rotation:Quat4f=$Engine.Owner.Rotation)
    WorldGridExtents:Vector2f=$Emitter.Grid2D_SetResolution.WorldGridExtents
   )
   /SetVariables(Emitter.Position:NiagaraPosition=&ConvertVectorToPosition(
     Input Position=$Emitter.Grid2D_CreateUnitToWorldTransform.LocalToWorldTranslation
     Conversion Operation:ENiagara_LWCConvertVecToPos=::NewEnumerator0
    )) {isSetParameters=true}
   /NiagaraFluids/Modules/Grid2D/Grid2D_SetBounds(
    EmitterProperties=DataInterface() {propertyValues {EmitterBinding {bindingMode=Self, emitterName=None}}}
    LocalPivot:Vector2f=$Emitter.Grid2D_CreateUnitToWorldTransform.LocalPivot
    WorldGridExtents:Vector2f=$Emitter.Grid2D_SetResolution.WorldGridExtents
   )
   /SetVariables(Emitter.OrthogonalVelocityMult=0.6) {isSetParameters=true}
   /NiagaraFluids/Modules/Grid2D/SplitMatrixIntoRows(Matrix:NiagaraMatrix=$Emitter.Grid2D_CreateUnitToWorldTransform.WorldToUnit)
   &SpawnRate(SpawnRate=5000, Use Spawn Probability=false)
   /Game/Effects/FlipSimulation/Niagara/Modules/Flip_SurfacingParameters(
    BaseColor_HighThickness:LinearColor=<rgba 0.0295 0.5777 1 1>
    BaseColor_LowThickness:LinearColor=<rgba 0 0.3333 1 1>
    CustomBaseColor_HighThickness=true
    CustomBaseColor_lowThickness=true
    CustomDarkFaceBrightness=false
    CustomEdgeOffset=true
    CustomEnvReflection_Tint=false
    CustomFakeSSS_Color=false
    CustomHDRIReflection_Tint=false
    CustomHueShiftDarkFace=false
    CustomHueShiftLightFace=false
    CustomLightFaceBrightness=false
    CustomRefractionStrength=false
    CustomScaleBlurDistance=true
    CustomShadowStrength=false
    CustomThicknessDivideNumber=true
    CustomThicknessEmissive_HighThickness=true
    CustomThicknessEmissive_LowThickness=true
    CustomTotalOpacityScale=true
    DebugBound=true
    Preset:ENiagara_FlipRenderPresets=::NewEnumerator1
    ScaleBlurDistance=0.8
    ThicknessDivide=60
    ThicknessEmissive_HighThickness:LinearColor=<rgba 0.0236 0.4562 0.8 1>
    ThicknessEmissive_LowThickness:LinearColor=<rgba 0 0.225 0.45 1>
    TotalOpacityScale:NiagaraFloat=&FloatFromCurve(
     FloatCurve=DataInterface() {propertyValues {Curve {keys [
         {interpMode=RCIM_Cubic, tangentMode=RCTM_Break, time=0, value=0}
         {arriveTangentWeight=0.0373, interpMode=RCIM_Cubic, leaveTangentWeight=0.3354, tangentMode=RCTM_User, time=0.05, value=1}
         <at 0.5 1>
        ]}}}
     CurveIndex=$System.Age
    )
   )
  ]
  particleSpawnScript [
   /Niagara/Modules/Spawn/Initialization/InitializeParticle(
    Color=<rgba 0 0.3383 1 1>
    Lifetime=&RandomRangeFloat(Minimum=0.45, Maximum=0.1)
    Module.Write Color=false
    Module.Write Lifetime=true
    Module.Write Mass=false
    Module.Write Material Random=true
    Module.Write Position=false
    Module.Write Scale=false
    Module.Write SpriteRotation=false
    Module.Write SpriteSize=true
    Position=&SimulationPosition()
    Sprite Size=&Vector2DFromFloat(Value=7)
   )
   /Game/Effects/FlipSimulation/Niagara/Modules/Flip_Source(
    DeltaTime:NiagaraFloat=$Emitter.DeltaTime
    EnableSimpleSourceMoving=true
    InheritVelocityScale=0
    Radius:NiagaraFloat=&FloatFromCurve(
     FloatCurve=DataInterface() {propertyValues {Curve {keys [{interpMode=RCIM_Cubic, tangentMode=RCTM_Break, time=0, value=0.3}, <at 0.6 1>]}}}
     CurveIndex=$System.Age
     Scale Curve=15
    )
    SourceAge=0.05
    SourceCount=10
    SourceMovingRemap=DataInterface() {propertyValues {Curve {keys [<at 0 0>,<at 1 1>]}}}
   )
   /Game/Effects/FlipSimulation/Niagara/Modules/Flip_SourceLinearMovement(InheritVelocityScale=5, LinearVelocityWS:Vector3f=<0 0 -5>)
   &CurlNoiseForce(
    Noise Frequency=4
    Noise Strength=&FloatFromCurve(
     FloatCurve=DataInterface() {propertyValues {Curve {keys [{interpMode=RCIM_Cubic, tangentMode=RCTM_Break, time=0, value=0}, <at 0.9 1>]}}}
     CurveIndex=&Multiply_Float(A=$System.Age, B=1)
     Scale Curve=200
    )
    Pan Noise=true
    Pan Noise Field=<0 0 5>
   )
   &ShapeLocation(MaskSpawnGroup=false, Override Seed=false, Sphere Radius=5, UseOwnerScale=true)
   &AddVelocity(Velocity=<0 0 0>)
   &AddVelocity(
    Cone Angle=200
    Inner Cone Angle=160
    Module.UseConeFalloff=false
    Module.UseVelDistribution=false
    Override Seed=false
    Velocity Mode=::NewEnumerator2
    Velocity Speed=100
   ) {module=AddVelocity001}
   &ApplyInitialForces()
   /SetVariables(Particles.Velocity=&Multiply_Vector(
     A=$Particles.Velocity
     B=&VectorFromFloat(Value=&ClampFloat(Float=&Power(a=&Divide_Float(A=0.0167, B=$Engine.DeltaTime), b=0.5), Min=0.5, Max=2))
    )) {isSetParameters=true}
   /SetVariables(Particles.LastEmitterPosition:NiagaraPosition=$Emitter.SampleSkeletalMesh.SampledPosition) {isSetParameters=true}
   &KillParticles(Kill Particles=&SetBoolByFloatComparison(A=&RandomRangeFloat(Minimum=0, Maximum=1), B=1)) {module=KillParticles001}
  ]
  particleUpdateScript [
   &ParticleState()
   &ScaleSpriteSize(Uniform Curve Sprite Scale=DataInterface() {propertyValues {Curve {
       defaultValue=0
       keys [
        {arriveTangent=-0.7327, interpMode=RCIM_Cubic, leaveTangent=-0.7327, tangentMode=RCTM_User, time=0, value=1}
        {arriveTangentWeight=1.2358, interpMode=RCIM_Cubic, tangentMode=RCTM_User, time=1, value=0}
       ]
      }}})
   &GravityForce(Gravity=<0 0 -700>)
   &PointForce(
    Force Origin=&AddVectorToPosition(Position=$Engine.Emitter.SimulationPosition, Vector=<0 0 0>)
    Force Strength=800
    Module.Use Falloff=true
    Module.UseLinearForce=false
   )
   &PointForce(
    Force Falloff Distance=70
    Force Origin=&AddVectorToPosition(Position=$Engine.Emitter.SimulationPosition, Vector=<0 0 -70>)
    Force Strength=-1500
    Module.Use Falloff=true
    Module.UseLinearForce=false
   ) {enabled=false, module=PointForce001}
   &Drag(Drag=3, Use Linear Drag=true, Use Rotational Drag=true)
   &CurlNoiseForce(
    Noise Strength=&FloatFromCurve(
     FloatCurve=DataInterface() {propertyValues {Curve {keys [{tangentMode=RCTM_User, time=0, value=1}, {tangentMode=RCTM_User, time=0.5, value=0}]}}}
     Scale Curve=500
    )
    Pan Noise=true
    Pan Noise Field=<1 1 1>
   ) {module=CurlNoiseForce001}
   /SetVariables(Transient.PhysicsForce:Vector3f=&Multiply_VectorByFloat(
     Vector=$Transient.PhysicsForce
     Float=&ClampFloat(
      Float=&Power(a=&Divide_Float(A=$Engine.DeltaTime, B=0.0167), b=0.5)
      Min=0.5
      Max=1.3
     )
    )) {isSetParameters=true}
   /SetVariables(Transient.StartPosition:NiagaraPosition=$Particles.Position) {isSetParameters=true}
   &InheritSourceMovement(ApplyForce=false, ApplyPosition=true, ApplyPositionFalloff=false)
   &SolveForcesAndVelocity(Clamp Velocity=false, DeltaTime:NiagaraFloat=$Emitter.DeltaTime, Limit Acceleration=false, Write to Presolve Properties=false)
   /SetVariables(Particles.Position:Vector3f=$Transient.StartPosition) {isSetParameters=true}
  ]
  rendererClasses [&NiagaraSpriteRendererProperties]
  renderers [
   {rendererClass=&NiagaraSpriteRendererProperties}
   {rendererClass=&NiagaraSpriteRendererProperties, rendererIndex=1}
   {rendererClass=&NiagaraSpriteRendererProperties, rendererIndex=2}
   {rendererClass=&NiagaraSpriteRendererProperties, rendererIndex=3}
   {rendererClass=&NiagaraSpriteRendererProperties, rendererIndex=4}
  ]
 }
}
summary {
 systemName=NS_SteelKite_FaceSplash
 emitters [
  {emitterName=Drips, rendererClasses [&NiagaraSpriteRendererProperties]}
  {emitterName=Drips001, rendererClasses [&NiagaraSpriteRendererProperties]}
  {emitterName=Drips002, rendererClasses [&NiagaraMeshRendererProperties]}
  {emitterName=FN_PIC_FLIP_Fluid_Stream002, simTarget=GPUComputeSim, rendererClasses [&NiagaraSpriteRendererProperties]}
  {emitterName=Drips003, rendererClasses [&NiagaraSpriteRendererProperties]}
 ]
 userVariables [
  User.ColorA {type=LinearColor, description="", defaultValue {struct=LinearColor, value=<rgba 1 1 1 1>}}
  User.ColorB {type=LinearColor, description="", defaultValue {struct=LinearColor, value=<rgba 1 1 1 1>}}
 ]
}
