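# Two-sided translucent unlit material for the glass of the SteelKite consumable, usable on skeletal meshes.
# A static switch selects either time-of-day tint darkened by sun-facing N·L with a minimum floor, or a flat color, output as exposure-compensated emissive.
# A texture mask blends toward near-black and opaque, while Fresnel sets transparency elsewhere.
# The reflection-color Add is unconnected, and the sun-plus-camera normal feeds an unlit model.
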Material /WildEstateConsumables/Gameplay/SteelKite/Materials/M_SteelKite_Glass
graph {
 properties {
  bAutomaticallySetUsageInEditor=false
  bTangentSpaceNormal=false
  bUseMaterialAttributes=true
  bUsedWithMeshDeformer=true
  bUsedWithSkeletalMesh=true
  blendMode=BLEND_Translucent
  shadingModel=MSM_Unlit
  twoSided=true
  usedShadingModels=Unlit
 }
 nodes [
  n1 = MaterialFunctionCall(
   Color_Morning=VectorParameter() {parameterName=Glass_Color_Morning, defaultValue=<rgba 0.8229 0.4438 0.1029 1>}.RGB
   Color_Day=VectorParameter() {parameterName=Glass_Color_Day, defaultValue=<rgba 0.1131 0.724 0.6119 1>}.RGB
   Color_Evening=VectorParameter() {parameterName=Glass_Color_Evening, defaultValue=<rgba 0.724 0.381 0.131 1>}.RGB
   Color_Night=VectorParameter() {parameterName=Glass_Color_Night, defaultValue=<rgba 0.0407 0.1246 0.224 1>}.RGB
  ) {materialFunction=&Color_Time_of_Day}
  Add(
   A=@n1.Result
   B=MaterialFunctionCall(
    Color_Morning=VectorParameter() {parameterName=Reflection_Color_Morning, defaultValue=<rgba 0.724 0.6384 0.4014 1>}.RGB
    Color_Day=VectorParameter() {parameterName=Reflection_Color_Day, defaultValue=<rgba 0.1131 0.724 0.6119 1>}.RGB
    Color_Evening=VectorParameter() {parameterName=Reflection_Color_Evening, defaultValue=<rgba 0.1131 0.724 0.6119 1>}.RGB
    Color_Night=VectorParameter() {parameterName=Reflection_Color_Night, defaultValue=<rgba 0.0254 0.1028 0.4427 1>}.RGB
   ) {materialFunction=&Color_Time_of_Day}.Result
  )
  n2 = CollectionParameter() {
   collection=&FortniteMaterialParameters
   parameterName=SunlightDirectionalVector
  }
  n3 = Multiply(A=VertexNormalWS(), B=TwoSidedSign())
  n4 = Saturate(None=Max(A=DotProduct(A=@n2, B=@n3), B=ScalarParameter() {defaultValue=0.4109, parameterName=MinDarkness}))
  n5 = TextureSample() {samplerType=SAMPLERTYPE_Alpha, texture=/WildEstateConsumables/Gameplay/SteelKite/Textures/T_SteelKite_Consumable_DFL}
  n6 = LinearInterpolate(
   A=StaticSwitchParameter(
    True=Multiply(A=@n1.Result, B=@n4)
    False=VectorParameter() {parameterName=BaseColorReflections, defaultValue=<rgba 0.4052 0.6545 0.8229 1>}.RGB
   ) {parameterName=TimeofDayGlass}
   Alpha=@n5.RGB
  ) {constB=0.005}
  n7 = ScalarParameter() {parameterName=Opacity-Fresnel Inner Transparency}
  n8 = ScalarParameter() {defaultValue=0.4, parameterName=Opacity-Fresnel Outer Transparency}
  n9 = Saturate(None=Fresnel(ExponentIn=ScalarParameter() {defaultValue=4, parameterName=Opacity-Fresnel Exponent}, Normal=VertexNormalWS()))
  n10 = CollectionParameter() {
   collection=&FortniteMaterialParameters
   parameterName=SunlightDirectionalVector
  }
  n11 = ComponentMask(None=@n10) {b=true, g=true, r=true}
  n12 = Transform(None=Constant3Vector() {constant=<rgba 0 0 -1 0>}) {transformSourceType=TRANSFORMSOURCE_Camera}
  SetMaterialAttributes!(
   MaterialAttributes
   Base Color=@n6
   Specular=Constant()
   Roughness=VertexColor().R
   Emissive Color=EyeAdaptationInverse(LightValueInput=@n6).EyeAdaptationInverse
   Opacity=LinearInterpolate(A=LinearInterpolate(A=@n7, B=@n8, Alpha=@n9), Alpha=@n5.RGB)
   Normal=Normalize(VectorInput=Add(A=@n11, B=@n12) {desc=Compromise for Direction light and all othe rlights})
  ) {attributeSetTypes [
    #1
    #2
    #3
    #4
    #5
    #6
   ]}
 ]
}
