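# Translucent-looking plastic lid material for the WildEstate SteelKite consumable, using Single Layer Water shading with skeletal mesh support.
# Outputs a flat tint, Fresnel-driven opacity and vertex-red roughness, with scattering and absorption fed zero.
# World-space normal blends sun direction with a camera axis as a lighting compromise.
# Virtual-plane window reflections, fill mask and absorption/scatter math remain in the graph without reaching the outputs.
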Material /WildEstateConsumables/Gameplay/SteelKite/Materials/M_Steelkite_Plastic_Lid
graph {
 properties {
  bAutomaticallySetUsageInEditor=false
  bTangentSpaceNormal=false
  bUseMaterialAttributes=true
  bUsedWithMeshDeformer=true
  bUsedWithSkeletalMesh=true
  shadingModel=MSM_SingleLayerWater
  usedShadingModels=SingleLayerWater
 }
 nodes [
  CameraVectorWS()
  n1 = CollectionParameter() {
   collection=&FortniteMaterialParameters
   parameterName=SunlightDirectionalVector
  }
  n2 = DotProduct(A=@n1, B=Multiply(A=VertexNormalWS(), B=TwoSidedSign()))
  Divide(A=Add(A=@n2))
  n3 = Subtract(A=SceneDepthWithoutWater!(UVs).R, B=PixelDepth().R)
  Divide(A=ScalarParameter() {parameterName=Absorption}, B=@n3)
  n4 = VectorParameter() {group=SingleLayerShading, parameterName=Absorption, defaultValue=<rgba 0.0084 0.026 0.9584 149.7252>}
  Divide(A=Divide(A=Constant() {r=1}, B=Multiply(A=@n4.RGB, B=@n4.RGB)), B=@n3)
  n5 = TransformPosition(None=Constant3Vector()) {transformSourceType=TRANSFORMPOSSOURCE_Instance, transformType=TRANSFORMPOSSOURCE_World}
  n6 = MaterialFunctionCall(Float3=Divide(A=@n5) {constB=100}) {materialFunction=&BreakOutFloat3Components}
  n7 = Frac(None=Add(A=Add(A=@n6.R, B=@n6.R), B=@n6.B))
  n8 = Transform(None=Constant3Vector() {constant=<rgba 0 0 1 0>}) {transformSourceType=TRANSFORMSOURCE_Camera}
  n9 = Transform(None=Constant3Vector() {constant=<rgba 1 0 0 0>}) {transformSourceType=TRANSFORMSOURCE_Camera}
  n10 = ObjectPositionWS()
  n11 = ReflectionVectorWS()
  n12 = Add(A=@n10, B=Multiply(A=Multiply(A=@n11) {constB=-1}, B=ScalarParameter() {defaultValue=15, parameterName=GlassRelectionOffset}))
  n13 = ScalarParameter() {defaultValue=150, parameterName=GlassRelectionTiling}
  n14 = TextureSample(UVs=Add(
    A=Multiply(A=@n7) {constB=33.3}
    B=MaterialFunctionCall(Plane Normal Axis=@n8, Plane X Axis=@n9, Plane Center=@n12, UV Scale=@n13) {materialFunction=&VirtualPlaneCoordinates}.UVs
   )) {texture=&/T_WildEstate_Windows}
  LinearInterpolate(
   A=ScalarParameter() {defaultValue=0.28, parameterName=Opacity_Lines}
   B=ScalarParameter() {defaultValue=0.664, parameterName=Opacity_Glass}
   Alpha=@n14.R
  )
  n15 = ScalarParameter() {defaultValue=4, parameterName=Opacity-Fresnel Exponent}
  n16 = VertexNormalWS()
  n17 = LinearInterpolate(
   A=ScalarParameter() {parameterName=Opacity-Fresnel Inner Transparency}
   B=ScalarParameter() {defaultValue=0.4, parameterName=Opacity-Fresnel Outer Transparency}
   Alpha=Saturate(None=Fresnel(ExponentIn=@n15, Normal=@n16))
  )
  LinearInterpolate(A=ScalarParameter() {defaultValue=0.8234, parameterName=Opacity_Fill}, B=@n17, Alpha=NamedRerouteUsage())
  n18 = VectorParameter() {parameterName=BaseColorReflections, defaultValue=<rgba 0.4052 0.6545 0.8229 1>}
  n19 = VectorParameter() {parameterName=BaseColorOpaque, defaultValue=<rgba 0.0423 0.3627 0.6771 1>}
  n20 = VectorParameter() {parameterName=BaseColorOpaqueFill, defaultValue=<rgba 0.0319 0.2734 0.5104 1>}
  n21 = NamedRerouteUsage()
  LinearInterpolate(
   A=@n18.RGB
   B=StaticSwitchParameter(True=LinearInterpolate(A=@n19.RGB, B=@n20.RGB, Alpha=@n21), False=@n19.RGB) {parameterName=UseWaterFillSimple}
   Alpha=@n14.R
  )
  Multiply(A=Multiply(A=@n18.RGB, B=OneMinus(None=@n14.R)), B=ScalarParameter() {defaultValue=0.5208, parameterName=EmissiveReflectionLines})
  n22 = VectorParameter() {group=SingleLayerShading, parameterName=Scatter, defaultValue=<rgba 0.257 0.2539 0.6632 -0.2284>}
  Multiply(A=@n22.RGB, B=Divide(A=@n22.A) {constB=100})
  n23 = Constant3Vector() {constant=<rgba 0 0 1 0>}
  n24 = Constant3Vector() {constant=<rgba 1 0 0 0>}
  n25 = TransformPosition(None=Constant3Vector()) {transformSourceType=TRANSFORMPOSSOURCE_Instance, transformType=TRANSFORMPOSSOURCE_World}
  n26 = ScalarParameter() {defaultValue=-12.1405, parameterName=FillHeight}
  n27 = Add(A=@n25, B=AppendVector(A=Constant2Vector(), B=@n26))
  n28 = ScalarParameter() {defaultValue=39.6541, parameterName=FillRadius}
  n29 = Subtract(A=WorldPosition().XYZ, B=@n25)
  n30 = Floor(None=Subtract(A=ComponentMask(None=@n29) {b=true}, B=@n26))
  n31 = Add(
   A=SphereMask(
    A=MaterialFunctionCall!(Plane Normal Axis=@n23, Plane X Axis=@n24, Plane Center=@n27, UV Scale=@n28) {materialFunction=/Game/Athena/Environments/Nexus/Materials/VirtualPlaneCoordinatesAbs}.UVs
    B=Constant2Vector() {g=0.5, r=0.5}
   ) {attenuationRadius=0.5}
   B=OneMinus(None=Saturate(None=@n30))
  )
  n32 = Round(None=Fresnel(ExponentIn=ScalarParameter() {defaultValue=1.4086, parameterName=FillFresnelEdge}, Normal=VertexNormalWS()) {baseReflectFraction=0})
  NamedRerouteDeclaration(None=Multiply(A=Saturate(None=@n31), B=OneMinus(None=@n32))) {name=FillMask, nodeColor=<rgba 0.9219 1 0 1>}
  Normalize(VectorInput=Subtract(A=@n5, B=CameraPositionWS()))
  Normalize(VectorInput=Subtract(A=@n10, B=CameraPositionWS()))
  Saturate(None=Max(A=@n2, B=ScalarParameter() {defaultValue=0.4109, parameterName=MinDarkness}))
  ScalarParameter() {parameterName=Scatter}
  n33 = CollectionParameter() {
   collection=&FortniteMaterialParameters
   parameterName=SunlightDirectionalVector
  }
  n34 = ComponentMask(None=@n33) {b=true, g=true, r=true}
  n35 = Transform(None=Constant3Vector() {constant=<rgba 0 0 -1 0>}) {transformSourceType=TRANSFORMSOURCE_Camera}
  SetMaterialAttributes!(
   MaterialAttributes
   Base Color=@n18.RGB
   Specular=Constant()
   Roughness=VertexColor().R
   Emissive Color
   Opacity=@n17
   Normal=Normalize(VectorInput=Add(A=@n34, B=@n35) {desc=Compromise for Direction light and all othe rlights})
  ) {attributeSetTypes [
    #1
    #2
    #3
    #4
    #5
    #6
   ]}
  n36 = Constant()
  SingleLayerWaterMaterialOutput!(ScatteringCoefficients=@n36, AbsorptionCoefficients=@n36, PhaseG, ColorScaleBehindWater)
  TextureCoordinate()
 ]
}
